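# First-person fire sound cue for the Athena rocket launcher in the Figment plugin.
# A Random node picks between two equally weighted rocket launcher start waves and feeds a root Modulator at default settings for slight variation.
# It uses shared first-person weapon projectile attenuation, the first-person ranged fire sound class and the first-person gunfire concurrency set.
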SoundCue /Figment_Launcher_Rocket_Athena/Sounds/FNS01_Weapon_LAU_Rocket_Launcher_Fire_1P_Athena_Cue
root=Modulator_0
nodes [
 {id=Modulator_0, children [Random_0]}
 {id=Random_0, children [WavePlayer_1,WavePlayer_0], properties {weights [1,1]}}
 {id=WavePlayer_0, properties {soundWaveAssetPtr=&/rocket_launcher_start_02}}
 {id=WavePlayer_1, properties {soundWaveAssetPtr=&/rocket_launcher_start_01}}
]
properties {
 attenuationSettings=&Fort_Weapon_Projectile_FirstPerson_Athena_Att
 bDebug=true
 maxDistance=2097152
 soundClassObject=&Fort_WeaponRanged_Fire_1P_PSM
 volumeMultiplier=1
 attenuationOverrides {manualReverbSendLevel=-1.0165, reverbSendMethod=Manual}
 concurrencyOverrides {bLimitToOwner=true, maxCount=3, resolutionRule=StopOldest, volumeScale=0.9}
 concurrencySet [&Weapon_Gunfire_FirstPerson_Conc]
}
