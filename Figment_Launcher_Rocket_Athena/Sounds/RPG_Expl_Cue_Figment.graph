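# Sound cue for the Figment rocket launcher's explosion, mixing close, medium and distant layers.
# Each layer has its own attenuation, modulator and sound class override, with the close layer on a 3P class and the others on a distant-fire class.
# Close and medium layers share one grenade explosion wave, while distant uses a separate sharp distant wave.
# A weapon explosion concurrency set is assigned.
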SoundCue /Figment_Launcher_Rocket_Athena/Sounds/RPG_Expl_Cue_Figment
root=Mixer_4
nodes [
 {
  id=Attenuation_0
  children [Modulator_1]
  properties {
   attenuationSettings=/Game/Sounds/Attenuation/Fort_Explosion_Distant_Athena_Att
   attenuationOverrides {falloffDistance=3000, attenuationShapeExtents {x=1000}}
  }
 }
 {
  id=Attenuation_4
  children [WavePlayer_10]
  properties {attenuationSettings=/Game/Sounds/Attenuation/Fort_Explosion_Close_Athena_Att, attenuationOverrides {falloffDistance=1600}}
 }
 {
  id=Attenuation_5
  children [Modulator_5]
  properties {
   attenuationSettings=/Game/Sounds/Attenuation/Fort_Explosion_Medium_Athena_Att
   attenuationOverrides {falloffDistance=3000, attenuationShapeExtents {x=1000}}
  }
 }
 {id=Mixer_4, children [SoundClass_3,SoundClass_2,SoundClass_0], properties {inputVolume [0.7,1,1.2]}}
 {id=Modulator_0, children [Attenuation_4]}
 {id=Modulator_1, children [WavePlayer_9], properties {pitchMax=1.1, pitchMin=0.9}}
 {id=Modulator_5, children [WavePlayer_10], properties {pitchMax=1.1, pitchMin=0.9}}
 {
  id=SoundClass_0
  children [Attenuation_0]
  properties {soundClassOverride=&Fort_WeaponRanged_Fire_Distant}
 }
 {
  id=SoundClass_2
  children [Attenuation_5]
  properties {soundClassOverride=&Fort_WeaponRanged_Fire_Distant}
 }
 {id=SoundClass_3, children [Modulator_0], properties {soundClassOverride=&Fort_WeaponRanged_Fire_3P_PSM}}
 {id=WavePlayer_10, properties {soundWaveAssetPtr=/Game/Sounds/Fort_GamePlay_Sounds/Explosives/Explosion_GrenadeProjectile_01_E}}
 {id=WavePlayer_9, properties {soundWaveAssetPtr=/Game/Sounds/Fort_GamePlay_Sounds/Explosives/explosion_sharp_distant_01}}
]
properties {
 maxDistance=27500
 volumeMultiplier=1
 concurrencyOverrides {maxCount=4, resolutionRule=StopOldest, volumeScale=0.5}
 concurrencySet [/Game/Sounds/Concurrency_Rules/Weapon_Explosion_Conc]
}
