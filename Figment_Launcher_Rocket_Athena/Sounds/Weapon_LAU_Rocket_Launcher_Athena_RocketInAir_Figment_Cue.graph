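# Sound cue for the in-flight sound of the Figment rocket launcher's projectile.
# A Mixer layers a looping in-air wave, faded in by an Enveloper, pitch-modulated and run through an exaggerated, smoothed Doppler, with a louder modulated projectile wave that bypasses the Doppler.
# It uses the launcher projectile in-air attenuation and overrides concurrency to cap instances per owner, stopping the quietest.
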SoundCue /Figment_Launcher_Rocket_Athena/Sounds/Weapon_LAU_Rocket_Launcher_Athena_RocketInAir_Figment_Cue
root=Mixer_1
nodes [
 {id=Doppler_1, children [Modulator_7], properties {bUseSmoothing=true, dopplerIntensity=3}}
 {
  id=Enveloper_4
  children [WavePlayer_5]
  properties {
   pitchCurve {editorCurveData {keys [{time=0}]}}
   volumeCurve {editorCurveData {keys [{interpMode=RCIM_Cubic, tangentMode=RCTM_User, time=0, value=0}, {interpMode=RCIM_Cubic, tangentMode=RCTM_User, time=1, value=1}]}}
  }
 }
 {id=Mixer_1, children [Doppler_1,Modulator_1], properties {inputVolume [1,1.5]}}
 {id=Modulator_1, children [WavePlayer_1]}
 {id=Modulator_7, children [Enveloper_4], properties {pitchMax=1.1}}
 {id=WavePlayer_1, properties {soundWaveAssetPtr=&/Athena_New_Rocket_Projectile_01}}
 {id=WavePlayer_5, properties {soundWaveAssetPtr=&/RPG_InAir_NewLoop_LP2}}
]
properties {
 attenuationSettings=/Game/Sounds/Attenuation/Fort_Weapon_Launcher_Proj_InAir_Default_Att
 bOverrideConcurrency=true
 maxDistance=12750
 soundClassObject=&Fort_WeaponRanged_Fire_3P_PSM
 volumeMultiplier=0.5
 attenuationOverrides {falloffDistance=5000, attenuationShapeExtents {x=1000}}
 concurrencyOverrides {bLimitToOwner=true, maxCount=3, resolutionRule=StopQuietest}
}
